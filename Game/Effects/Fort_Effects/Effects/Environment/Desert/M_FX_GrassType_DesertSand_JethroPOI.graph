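# Translucent unlit material for instanced desert sand wisps, likely spawned as landscape grass at the Jethro POI, tinted by time of day.
# World position offset combines SplineThicken, a looping per-instance drift and height conforming to the Apollo landscape runtime virtual texture.
# Opacity multiplies an edge gradient, panning distorted smoke textures, depth, camera and instance fades, and a pulse synced to the drift.
# A Fresnel on the landscape RVT normal further modulates opacity.
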Material /Game/Effects/Fort_Effects/Effects/Environment/Desert/M_FX_GrassType_DesertSand_JethroPOI
graph {
 properties {
  bUseMaterialAttributes=true
  bUsedWithInstancedStaticMeshes=true
  blendMode=BLEND_Translucent
  pixelDepthOffsetMode=PDOM_Legacy
  shadingModel=MSM_Unlit
  translucencyLightingMode=TLM_VolumetricPerVertexDirectional
  twoSided=true
  usedShadingModels=Unlit
 }
 nodes [
  Divide()
  Length(None=WorldPosition() {worldPositionShaderOffset=WPT_CameraRelative}.XYZ) {is=class MaterialExpressionLength}
  MaterialFunctionCall(Distance=ScalarParameter() {defaultValue=100, parameterName=CameraScaleDistance}) {materialFunction=&Near_Camera_Fade}
  Min()
  Min()
  n1 = ScalarParameter() {parameterName=Cam Fade length Close}
  n2 = ScalarParameter() {parameterName=Cam Fade Offset close}
  n3 = MaterialFunctionCall(Fade Length=ScalarParameter() {parameterName=Cam Fade length Far}, Fade Offset=ScalarParameter() {parameterName=Cam Fade Offset Far}) {materialFunction=&CameraDepthFade}
  n4 = Saturate(None=Multiply(
    A=MaterialFunctionCall(Fade Length=@n1, Fade Offset=@n2) {materialFunction=&CameraDepthFade}.Result
    B=OneMinus(None=@n3.Result)
   ))
  n5 = Clamp(None=@n4, Min=ScalarParameter() {defaultValue=0.1, parameterName=Clamp Min}) {minDefault=0.1}
  Multiply(B=@n5)
  Multiply(B=@n5)
  NamedRerouteDeclaration(None=TransformPosition(None=ObjectPositionWS()) {transformSourceType=TRANSFORMPOSSOURCE_Instance, transformType=TRANSFORMPOSSOURCE_Instance}) {name=Xformd Obj Pos, nodeColor=<rgba 0.9453 0 1 1>}
  ObjectPositionWS()
  ObjectPositionWS()
  Saturate()
  n6 = ScalarParameter() {parameterName=Cam Fade length Close}
  n7 = ScalarParameter() {parameterName=Cam Fade Offset close}
  n8 = MaterialFunctionCall(Fade Length=ScalarParameter() {parameterName=Cam Fade length Far}, Fade Offset=ScalarParameter() {parameterName=Cam Fade Offset Far}) {materialFunction=&CameraDepthFade}
  n9 = VectorParameter() {parameterName=RVTHeightConform, defaultValue=<rgba 0 0 1 1>}
  n10 = RuntimeVirtualTextureSample!(UVs, World Position) {materialType=WorldHeight, virtualTexture=/Game/Athena/Apollo/Maps/Landscape/VT/RVT_Apollo_Landscape_Height}
  n11 = ComponentMask(None=WorldPosition().XYZ) {b=true}
  n12 = ComponentMask(None=ObjectBounds()) {b=true}
  n13 = Subtract(A=Add(A=Subtract(A=@n10.WorldHeight, B=@n11), B=@n12), B=Multiply(A=@n12, B=ScalarParameter() {defaultValue=1, parameterName=OffsetHeight}))
  n14 = MaterialFunctionCall(
   WidthBase=ScalarParameter() {defaultValue=10, parameterName="WidthBase (S)"}
   WidthTip=ScalarParameter() {defaultValue=1, parameterName="WidthTip (S)"}
   WorldPosition=WorldPosition().XYZ
  ) {materialFunction=&SplineThicken}
  n15 = ScalarParameter() {defaultValue=1, parameterName=ObjectRadiusMult}
  n16 = Divide(A=Time() {bOverride_Period=true, period=6}) {constB=6}
  n17 = PerInstanceRandom()
  n18 = LinearInterpolate(A=Multiply(A=@n15) {constB=-1}, B=@n15, Alpha=Frac(None=Add(A=@n16, B=@n17)))
  n19 = Multiply(
   A=Multiply(A=@n18, B=ScalarParameter() {defaultValue=1, group=WPO, parameterName=OverallSpeed})
   B=LinearInterpolate(
    A=VectorParameter() {parameterName=OverallDirection, defaultValue=<rgba 1 0 0 1>}.RGB
    B=VectorParameter() {parameterName=OverallDirectionMax, defaultValue=<rgba 1 0 0 1>}.RGB
    Alpha=PerInstanceRandom()
   )
  )
  n20 = ComponentMask(None=@n19) {r=true}
  n21 = ComponentMask(None=@n19) {g=true}
  n22 = ScalarParameter() {defaultValue=1, group=Movement, parameterName=Curl_Amount}
  n23 = ScalarParameter() {defaultValue=1, group=Movement, parameterName=Curl_Speed}
  n24 = Multiply(A=@n23) {constA=0.65, constB=0.65}
  n25 = PerInstanceRandom()
  n26 = Subtract(A=Subtract(A=WorldPosition().XYZ, B=ActorPositionWS()), B=Multiply(A=LinearInterpolate(A=@n23, B=@n24, Alpha=@n25), B=Time()))
  n27 = ScalarParameter() {defaultValue=1, group=Movement, parameterName=Curl_Scale}
  n28 = LinearInterpolate(A=@n27, B=Multiply(A=@n27) {constA=0.65, constB=0.65}, Alpha=PerInstanceRandom())
  n29 = Multiply(
   A=LinearInterpolate(A=@n22, B=Multiply(A=@n22) {constA=0.65, constB=0.65}, Alpha=PerInstanceRandom())
   B=VectorNoise(World Position=Divide(A=@n26, B=@n28)) {noiseFunction=VNF_VectorALU}
  )
  n30 = Constant()
  n31 = FeatureLevelSwitch(Default=@n29, ES2=@n30, ES3_1=@n30, SM4_REMOVED=@n29)
  n32 = Add(A=ShadingPathSwitch(Default=QualitySwitch(Default=@n31, Low=@n30, Medium=@n30), Forward=@n30), B=@n19)
  n33 = ActorPositionWS()
  n34 = Constant() {r=1}
  n35 = FeatureLevelSwitch(
   Default=RuntimeVirtualTextureSample(World Position=Add(A=@n32, B=@n33)) {materialType=WorldHeight, virtualTexture=&RVT_Apollo_Landscape_Height}.WorldHeight
   ES2=@n34
   ES3_1=@n34
  )
  n36 = Subtract(
   A=QualitySwitch(Default=ShadingPathSwitch(Default=@n35, Forward=@n34, Mobile=@n34), Low=@n34)
   B=ComponentMask(None=ActorPositionWS()) {b=true}
  )
  n37 = ComponentMask(None=ObjectBounds()) {b=true}
  n38 = ScalarParameter() {defaultValue=1, parameterName=OffsetHeight}
  n39 = Multiply(
   A=VectorParameter() {parameterName=ConformMult, defaultValue=<rgba 0 0 1 1>}.RGB
   B=Subtract(A=Add(A=@n36, B=@n37), B=Multiply(A=@n37, B=@n38))
  )
  n40 = Add(
   A=Add(A=Multiply(A=@n9.RGB, B=@n13), B=@n14.WorldPositionOffset)
   B=AppendVector(A=AppendVector(A=@n20, B=@n21), B=ComponentMask(None=@n39) {b=true})
  )
  SetMaterialAttributes!(
   MaterialAttributes
   Emissive Color=Multiply(
    A=MaterialFunctionCall(Fade Length=@n6, Fade Offset=@n7) {materialFunction=&CameraDepthFade}.Result
    B=OneMinus(None=@n8.Result)
   )
   Opacity=Constant() {r=1}
   World Position Offset=@n40
   Normal
  ) {attributeSetTypes [
    #1
    #2
    #3
    #4
   ]}
  n41 = VectorParameter() {group=TOD, parameterName="Color_Morning (V3)", defaultValue=<rgba 1 1 1 0>}
  n42 = VectorParameter() {group=TOD, parameterName="Color_Day (V3)", defaultValue=<rgba 1 1 1 0>}
  n43 = VectorParameter() {group=TOD, parameterName="Color_Evening (V3)", defaultValue=<rgba 1 1 1 0>}
  n44 = VectorParameter() {group=TOD, parameterName="Color_Night (V3)", defaultValue=<rgba 1 1 1 0>}
  n45 = MaterialFunctionCall() {materialFunction=&LinearGradient}
  n46 = Multiply(A=Multiply(A=OneMinus(None=@n45.UGradient), B=@n45.UGradient), B=Multiply(A=@n45.VGradient, B=OneMinus(None=@n45.VGradient)))
  n47 = ScalarParameter() {defaultValue=2, parameterName=GradientAmount}
  n48 = ScalarParameter() {defaultValue=1, parameterName=TileX_Main}
  n49 = ScalarParameter() {defaultValue=1, parameterName=TileX2_Main}
  n50 = PerInstanceRandom()
  n51 = TextureCoordinate() {coordinateIndex=1}
  n52 = AppendVector(
   A=Multiply(A=LinearInterpolate(A=@n48, B=@n49, Alpha=@n50) {constA=0.1, constB=0.25}, B=ComponentMask(None=@n51) {r=true})
   B=Multiply(A=ScalarParameter() {defaultValue=1, parameterName=TileY_Main}, B=ComponentMask(None=@n51) {g=true})
  )
  n53 = AppendVector(
   A=LinearInterpolate(
    A=ScalarParameter() {defaultValue=0.15, parameterName=SpeedX}
    B=ScalarParameter() {defaultValue=0.15, parameterName=SpeedX2}
    Alpha=PerInstanceRandom()
   ) {constA=0.1, constB=0.25}
   B=LinearInterpolate(
    A=ScalarParameter() {defaultValue=0.15, parameterName=SpeedY}
    B=ScalarParameter() {defaultValue=0.15, parameterName=SpeedY2}
    Alpha=PerInstanceRandom()
   ) {constA=0.1, constB=0.25}
  )
  n54 = TextureSampleParameter2D(UVs=VertexInterpolator(VS=Panner(Coordinate=@n52, Speed=@n53) {speedX=0.15}).PS) {parameterName=MainTexture, texture=/Game/Effects/Fort_Effects/Textures/Smoke/T_Puffy_Smoke_Trail_01}
  n55 = TextureCoordinate()
  n56 = LinearInterpolate(Alpha=PerInstanceRandom()) {constA=0.5}
  n57 = AppendVector(
   A=Multiply(A=@n56, B=ScalarParameter() {defaultValue=1, group=Texture values, parameterName=TileX})
   B=ScalarParameter() {defaultValue=1, group=Texture values, parameterName=TileY}
  )
  n58 = ScalarParameter() {defaultValue=0.1, group=Texture values, parameterName=SpeedMult}
  n59 = Panner(
   Coordinate=Multiply(A=@n55, B=@n57)
   Speed=AppendVector(A=ScalarParameter() {defaultValue=0.15, group=Texture values, parameterName=SpeedY}, B=Multiply(A=@n56, B=@n58))
  ) {speedY=1}
  n60 = TextureSampleParameter2D(UVs=Multiply(A=TextureCoordinate(), B=@n57)) {
   parameterName=DistortionTexture
   samplerType=SAMPLERTYPE_LinearColor
   texture=/Game/Global/Textures/Default/Tileable_Patterns/T_Tile_Smoke_Small
  }
  n61 = TextureSampleParameter2D(UVs=Multiply(A=TextureCoordinate(), B=@n57)) {
   parameterName=DistortionTexture
   samplerType=SAMPLERTYPE_LinearColor
   texture=&T_Tile_Smoke_Small
  }
  n62 = Add(A=ComponentMask(None=@n60.RGB) {g=true, r=true}, B=ComponentMask(None=@n61.RGB) {g=true, r=true})
  n63 = ScalarParameter() {defaultValue=0.1, group=Texture values, parameterName=BaseDistort}
  n64 = Constant()
  n65 = QualitySwitch(Default=ComponentMask(None=Multiply(A=@n62, B=@n63)) {g=true, r=true}, Low=@n64)
  n66 = FeatureLevelSwitch(Default=@n65, ES2=@n64, ES3_1=@n64, SM4_REMOVED=@n65)
  n67 = TextureSampleParameter2D(UVs=VertexInterpolator(VS=Add(A=@n59, B=@n66)).PS) {parameterName=SmokeTexture, texture=/Game/Global/Textures/Default/Caustics/T_Turbulence_Seamless}
  n68 = ScalarParameter() {defaultValue=0.1, group=Texture values, parameterName=SpeedMult_02}
  n69 = Panner(
   Coordinate=Multiply(A=@n55, B=@n57)
   Speed=AppendVector(A=ScalarParameter() {defaultValue=0.15, group=Texture values, parameterName=SpeedY}, B=Multiply(A=@n56, B=@n68))
  ) {speedY=1}
  n70 = TextureSampleParameter2D(UVs=VertexInterpolator(VS=Add(A=@n69, B=@n66)).PS) {parameterName=SmokeTexture, texture=&T_Turbulence_Seamless}
  n71 = LinearInterpolate(
   A=Constant() {r=1}
   B=Power(Base=Add(A=@n67.G, B=@n70.G), Exp=ScalarParameter() {defaultValue=0.15, parameterName=Contrast})
   Alpha=ScalarParameter() {defaultValue=1, parameterName=ExtraTextureAmount}
  )
  n72 = DepthFade(
   Opacity=Multiply(A=Multiply(A=@n46, B=@n47), B=Multiply(A=@n54.G, B=@n71))
   FadeDistance=ScalarParameter() {defaultValue=25, parameterName=DepthFadeDistance}
  )
  n73 = LinearInterpolate(A=ScalarParameter() {defaultValue=1, parameterName=Opacity}, B=ScalarParameter() {defaultValue=1, parameterName=OpacityFar}, Alpha=@n4)
  n74 = Sine(None=Add(A=@n18) {constB=-0.25})
  n75 = Saturate(None=LinearInterpolate(
    A=ScalarParameter() {parameterName=fade low}
    B=ScalarParameter() {defaultValue=1, parameterName=Fade Hight}
    Alpha=ConstantBiasScale(None=@n74)
   ))
  n76 = ScalarParameter() {defaultValue=0.6, parameterName=Fade pow}
  n77 = Multiply(
   A=Multiply(A=Multiply(A=@n72, B=@n73), B=PerInstanceFadeAmount())
   B=VertexInterpolator(VS=Power(Base=@n75, Exp=@n76) {constExponent=0.6}).PS
  )
  n78 = RuntimeVirtualTextureSample!(UVs, World Position, Mip Bias=ScalarParameter() {defaultValue=5, parameterName=normal Mip Test}) {
   materialType=BaseColor_Normal_Specular_YCoCg
   mipValueMode=RVTMVM_MipBias
   virtualTexture=/Game/Athena/Apollo/Maps/Landscape/VT/RVT_Apollo_Landscape
  }
  n79 = Saturate(None=LinearInterpolate(
    A=ScalarParameter() {defaultValue=1, parameterName=Fres Low}
    B=ScalarParameter() {defaultValue=10, parameterName=Fres HIgh}
    Alpha=Fresnel(Normal=@n78.Normal) {baseReflectFraction=0, exponent=1}
   ) {constA=1, constB=0})
  SetMaterialAttributes(
   Emissive Color=VertexInterpolator(VS=MaterialFunctionCall(Color_Morning=@n41.RGB, Color_Day=@n42.RGB, Color_Evening=@n43.RGB, Color_Night=@n44.RGB) {materialFunction=&Color_Time_of_Day}.Result).PS
   Opacity=Multiply(A=Multiply(A=@n77, B=@n79), B=ScalarParameter() {defaultValue=0.7, parameterName=Master Opacity})
   World Position Offset=@n40
   Normal=@n14.Normal
  ) {attributeSetTypes [
    #1
    #2
    #3
    #4
   ]}
  Subtract()
 ]
}
